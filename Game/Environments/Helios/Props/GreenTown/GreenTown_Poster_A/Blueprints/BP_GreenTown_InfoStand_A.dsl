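; GreenTown info stand prop Blueprint that shows randomly chosen posters on its three poster mesh components.
; The construction script gets an array from SwatchList, then picks a random BuildingTextureData entry independently for each poster, so repeats are possible.
; Each pick's diffuse texture goes into the "Diffuse" parameter of that component's dynamic material instance, with no seed for the randomness.
; The event graph holds only empty stub events.
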(event EventBeginPlay
  (|CallParentFunction) ; This node is disabled and will not be called.\nDrag off pins to build functionality.
)

(event Collision|EventActorBeginOverlap
  (params (OtherActor :type Actor%20Object%20Reference)))

(event EventTick
  (params (DeltaSeconds :type Float%20%28single-precision%29)))

(fn ConstructionScript
  (|CallParentFunction)
  (bind _color_swatches (SwatchList|GetSwatchArray))
  (for _array_element (Utilities|Array|MakeArray (Variables|Default|GetSM_Commerce_InfoStand_Poster_A1) (Variables|Default|GetSM_Commerce_InfoStand_Poster_A2) (Variables|Default|GetSM_Commerce_InfoStand_Poster_A))
    (Rendering|Material|SetTextureParameterValue (Class|BTechArtGlobalLibrary|GetDynamicMaterialInstance _array_element) "Diffuse" (Variables|/Script/FortniteGame.BuildingTextureData|GetDiffuse (Utilities|Array|RandomArrayItem _color_swatches)))))
